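# Physics asset for the with-backpack variant of the M_MED turtleneck cape cosmetic, defining simulation bodies and joints for the cape bones.
# Left, center and right capsule chains hang from master bones constrained to clavicle_r, with a center branch at c_05.
# Limited joints keep twist narrow and generally widen swing toward the hem, while free-rotation constraints tie the three top segments together.
# Torso, head and right-arm capsules plus an unconstrained crouch sphere supply body shapes around the cape.
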PhysicsAsset /BRCosmetics/Accessories/FORT_Capes/Cape_M_MED_Turtleneck/Meshes/M_MED_Turtleneck_Cape_WithBackpack_Physics
bodies {
 clavicle_r [{shapeType=Capsule, length=24.2119, radius=6.1031, center=<-13.0838 0.0001 0>, rotation=<rot 90 0 0>}]
 dyn_cape01_l_MSTR [{shapeType=Capsule, length=1, radius=1, rotation=<rot 90 0 0>}]
 dyn_cape_c_01 [{shapeType=Capsule, length=12.029, radius=3.8751, center=<6.5811 0 0.1231>, rotation=<rot 0 0 90>}]
 dyn_cape_c_02 [{shapeType=Capsule, length=11.5047, radius=4.3995, center=<6.5811 0.0004 0.3643>, rotation=<rot 0 0 90>}]
 dyn_cape_c_03 [{shapeType=Capsule, length=22.4699, radius=5.2507, center=<6.5811 0 -0.0198>, rotation=<rot 0 0 90>}]
 dyn_cape_c_04 [{shapeType=Capsule, length=29.8342, radius=7.1367, center=<6.5811 -0.0003 -1.8578>, rotation=<rot 0 0 90>}]
 dyn_cape_c_05 [{shapeType=Capsule, length=23.9595, radius=3.8702, center=<6.5811 -0.0001 -0.0115>, rotation=<rot 0 0 90>}]
 dyn_cape_c_06 [{shapeType=Capsule, length=20.4599, radius=7.3698, center=<6.5811 -0.0002 0.0495>, rotation=<rot 0 0 90>}]
 dyn_cape_c_07 [{shapeType=Capsule, length=19.9675, radius=7.8622, center=<6.5811 0 0.0237>, rotation=<rot 0 0 90>}]
 dyn_cape_c_08 [{shapeType=Capsule, length=19.9675, radius=7.8622, center=<6.5811 0 0.1231>, rotation=<rot 0 0 90>}]
 dyn_cape_c_09 [{shapeType=Capsule, length=20.8362, radius=6.9935, center=<6.5811 0 0.1231>, rotation=<rot 0 0 90>}]
 dyn_cape_c_10 [{shapeType=Capsule, length=19.9675, radius=7.8622, center=<6.5811 0 0.1231>, rotation=<rot 0 0 90>}]
 dyn_cape_c_MSTR [{shapeType=Capsule, length=1, radius=1, rotation=<rot 90 0 0>}]
 dyn_cape_l_01 [{shapeType=Capsule, length=12.7383, radius=3.1658, center=<6.5811 0.0001 -0.7553>, rotation=<rot 0 0 90>}]
 dyn_cape_l_02 [{shapeType=Capsule, length=12.214, radius=3.6902, center=<6.5811 0.0004 -0.514>, rotation=<rot 0 0 90>}]
 dyn_cape_l_03 [{shapeType=Capsule, length=21.6869, radius=6.0337, center=<6.5811 0.0003 -0.7432>, rotation=<rot 0 0 90>}]
 dyn_cape_l_04 [{shapeType=Capsule, length=31.1878, radius=5.7831, center=<6.5946 0.4808 -0.8649>, rotation=<rot -1.2452 -7.114 77.304>}]
 dyn_cape_l_05 [{shapeType=Capsule, length=25.2937, radius=2.536, center=<6.5811 0 -0.5813>, rotation=<rot 0 0 90>}]
 dyn_cape_l_06 [{shapeType=Capsule, length=22.1286, radius=5.7011, center=<6.5811 0.1083 -0.6809>, rotation=<rot 0 0 70.8176>}]
 dyn_cape_r_01 [{shapeType=Capsule, length=11.5208, radius=4.3834, center=<6.5811 0 1.3912>, rotation=<rot 0 0 90>}]
 dyn_cape_r_02 [{shapeType=Capsule, length=10.9965, radius=4.9077, center=<6.5811 0.0003 1.6324>, rotation=<rot 0 0 90>}]
 dyn_cape_r_03 [{shapeType=Capsule, length=22.2131, radius=5.5076, center=<6.5811 0 0.0182>, rotation=<rot 0 0 90>}]
 dyn_cape_r_04 [{shapeType=Capsule, length=27.6692, radius=9.3018, center=<6.5811 0.4125 -0.9053>, rotation=<rot 0.6574 7.4531 84.9867>}]
 dyn_cape_r_05 [{shapeType=Capsule, length=21.0921, radius=6.7376, center=<6.5811 -0.0002 0.0047>, rotation=<rot 0 0 90>}]
 dyn_cape_r_06 [{shapeType=Capsule, length=20.4599, radius=7.3698, center=<6.5811 -0.0002 -0.0653>, rotation=<rot 0 0 90>}]
 dyn_cape_r_07 [{shapeType=Capsule, length=19.9675, radius=7.8622, center=<6.5811 0 0.0146>, rotation=<rot 0 0 90>}]
 dyn_cape_r_08 [{shapeType=Capsule, length=19.9675, radius=7.8622, center=<6.5811 -0.0001 -1.138>, rotation=<rot 0 0 90>}]
 dyn_cape_r_MSTR [{shapeType=Capsule, length=1, radius=1, rotation=<rot 90 0 0>}]
 dyn_crouch_collision [{shapeType=Sphere, radius=51.125, center=<-0.0006 6.727 0>}]
 head [{shapeType=Capsule, length=8.75, radius=6.5}]
 neck_01 [{shapeType=Capsule, length=8.75, radius=6.5}]
 neck_02 [{shapeType=Capsule, length=8.75, radius=6.5}]
 pelvis [{shapeType=Capsule, length=26.8304, radius=16.1708, center=<-0.3186 -0.218 -0.0005>}]
 spine_01 [{shapeType=Capsule, length=34.5854, radius=10.2235}]
 spine_02 [{shapeType=Capsule, length=27.451, radius=10.2614}]
 spine_03 [{shapeType=Capsule, length=20.1297, radius=10.2314}]
 spine_04 [{shapeType=Capsule, length=17.3684, radius=12.9928}]
 spine_05 [{shapeType=Capsule, length=20.0158, radius=10.3454}]
 upperarm_r [{shapeType=Capsule, length=9.0457, radius=10.9543, center=<-12.8125 0 0>, rotation=<rot 90 0 0>}]
 upperarm_twist_01_r [{shapeType=Capsule, length=11.5296, radius=8.4704, center=<-12.8125 0 0>, rotation=<rot 90 0 0>}]
}
constraints [
 {bone1Name=dyn_cape_c_01, bone2Name=dyn_cape_c_MSTR, swing1LimitDegrees=10, swing2LimitDegrees=10, twistLimitDegrees=5}
 {bone1Name=dyn_cape_c_02, bone2Name=dyn_cape_c_01, swing1LimitDegrees=20, swing2LimitDegrees=20, twistLimitDegrees=5}
 {bone1Name=dyn_cape_c_03, bone2Name=dyn_cape_c_02, swing1LimitDegrees=5, swing2LimitDegrees=5, twistLimitDegrees=5}
 {bone1Name=dyn_cape_c_04, bone2Name=dyn_cape_c_03, swing1LimitDegrees=40, swing2LimitDegrees=40, twistLimitDegrees=5}
 {bone1Name=dyn_cape_c_05, bone2Name=dyn_cape_c_04, swing1LimitDegrees=45, swing2LimitDegrees=45, twistLimitDegrees=5}
 {bone1Name=dyn_cape_c_06, bone2Name=dyn_cape_c_05, swing1LimitDegrees=45, swing2LimitDegrees=45, twistLimitDegrees=5}
 {bone1Name=dyn_cape_c_07, bone2Name=dyn_cape_c_06, swing1LimitDegrees=45, swing2LimitDegrees=45, twistLimitDegrees=5}
 {bone1Name=dyn_cape_c_08, bone2Name=dyn_cape_c_07, swing1LimitDegrees=45, swing2LimitDegrees=45, twistLimitDegrees=5}
 {bone1Name=dyn_cape_c_09, bone2Name=dyn_cape_c_05, swing1LimitDegrees=45, swing2LimitDegrees=45, twistLimitDegrees=5}
 {bone1Name=dyn_cape_c_10, bone2Name=dyn_cape_c_09, swing1LimitDegrees=45, swing2LimitDegrees=45, twistLimitDegrees=5}
 {bone1Name=dyn_cape_r_01, bone2Name=dyn_cape_r_MSTR, swing1LimitDegrees=10, swing2LimitDegrees=10, twistLimitDegrees=5}
 {bone1Name=dyn_cape_r_02, bone2Name=dyn_cape_r_01, swing1LimitDegrees=20, swing2LimitDegrees=20, twistLimitDegrees=5}
 {bone1Name=dyn_cape_r_03, bone2Name=dyn_cape_r_02, swing1LimitDegrees=5, swing2LimitDegrees=5, twistLimitDegrees=5}
 {bone1Name=dyn_cape_r_04, bone2Name=dyn_cape_r_03, swing1LimitDegrees=40, swing2LimitDegrees=40, twistLimitDegrees=5}
 {bone1Name=dyn_cape_r_05, bone2Name=dyn_cape_r_04, swing1LimitDegrees=45, swing2LimitDegrees=45, twistLimitDegrees=5}
 {bone1Name=dyn_cape_r_06, bone2Name=dyn_cape_r_05, swing1LimitDegrees=45, swing2LimitDegrees=45, twistLimitDegrees=5}
 {bone1Name=dyn_cape_r_07, bone2Name=dyn_cape_r_06, swing1LimitDegrees=45, swing2LimitDegrees=45, twistLimitDegrees=5}
 {bone1Name=dyn_cape_r_08, bone2Name=dyn_cape_r_07, swing1LimitDegrees=45, swing2LimitDegrees=45, twistLimitDegrees=5}
 {bone1Name=dyn_cape_l_01, bone2Name=dyn_cape01_l_MSTR, swing1LimitDegrees=10, swing2LimitDegrees=10, twistLimitDegrees=5}
 {bone1Name=dyn_cape_l_02, bone2Name=dyn_cape_l_01, swing1LimitDegrees=20, swing2LimitDegrees=20, twistLimitDegrees=5}
 {bone1Name=dyn_cape_l_03, bone2Name=dyn_cape_l_02, swing1LimitDegrees=5, swing2LimitDegrees=5, twistLimitDegrees=5}
 {bone1Name=dyn_cape_l_04, bone2Name=dyn_cape_l_03, swing1LimitDegrees=40, swing2LimitDegrees=40, twistLimitDegrees=5}
 {bone1Name=dyn_cape_l_05, bone2Name=dyn_cape_l_04, swing1LimitDegrees=45, swing2LimitDegrees=45, twistLimitDegrees=5}
 {bone1Name=dyn_cape_l_06, bone2Name=dyn_cape_l_05, swing1LimitDegrees=45, swing2LimitDegrees=45, twistLimitDegrees=5}
 {
  bone1Name=dyn_cape_c_MSTR
  bone2Name=clavicle_r
  swing1LimitDegrees=45
  swing1Motion=Free
  swing2LimitDegrees=45
  swing2Motion=Free
  twistLimitDegrees=45
  twistMotion=Free
 }
 {
  bone1Name=dyn_cape_r_MSTR
  bone2Name=clavicle_r
  swing1LimitDegrees=45
  swing1Motion=Free
  swing2LimitDegrees=45
  swing2Motion=Free
  twistLimitDegrees=45
  twistMotion=Free
 }
 {
  bone1Name=dyn_cape01_l_MSTR
  bone2Name=clavicle_r
  swing1LimitDegrees=45
  swing1Motion=Free
  swing2LimitDegrees=45
  swing2Motion=Free
  twistLimitDegrees=45
  twistMotion=Free
 }
 {
  bone1Name=spine_01
  bone2Name=pelvis
  swing1LimitDegrees=45
  swing1Motion=Free
  swing2LimitDegrees=45
  swing2Motion=Free
  twistLimitDegrees=45
  twistMotion=Free
 }
 {
  bone1Name=spine_02
  bone2Name=spine_01
  swing1LimitDegrees=45
  swing1Motion=Free
  swing2LimitDegrees=45
  swing2Motion=Free
  twistLimitDegrees=45
  twistMotion=Free
 }
 {
  bone1Name=spine_03
  bone2Name=spine_02
  swing1LimitDegrees=45
  swing1Motion=Free
  swing2LimitDegrees=45
  swing2Motion=Free
  twistLimitDegrees=45
  twistMotion=Free
 }
 {
  bone1Name=spine_04
  bone2Name=spine_03
  swing1LimitDegrees=45
  swing1Motion=Free
  swing2LimitDegrees=45
  swing2Motion=Free
  twistLimitDegrees=45
  twistMotion=Free
 }
 {
  bone1Name=spine_05
  bone2Name=spine_04
  swing1LimitDegrees=45
  swing1Motion=Free
  swing2LimitDegrees=45
  swing2Motion=Free
  twistLimitDegrees=45
  twistMotion=Free
 }
 {
  bone1Name=clavicle_r
  bone2Name=spine_05
  swing1LimitDegrees=45
  swing1Motion=Free
  swing2LimitDegrees=45
  swing2Motion=Free
  twistLimitDegrees=45
  twistMotion=Free
 }
 {
  bone1Name=upperarm_r
  bone2Name=clavicle_r
  swing1LimitDegrees=45
  swing1Motion=Free
  swing2LimitDegrees=45
  swing2Motion=Free
  twistLimitDegrees=45
  twistMotion=Free
 }
 {
  bone1Name=upperarm_twist_01_r
  bone2Name=upperarm_r
  swing1LimitDegrees=45
  swing1Motion=Free
  swing2LimitDegrees=45
  swing2Motion=Free
  twistLimitDegrees=45
  twistMotion=Free
 }
 {
  bone1Name=neck_01
  bone2Name=spine_05
  swing1LimitDegrees=45
  swing1Motion=Free
  swing2LimitDegrees=45
  swing2Motion=Free
  twistLimitDegrees=45
  twistMotion=Free
 }
 {
  bone1Name=neck_02
  bone2Name=neck_01
  swing1LimitDegrees=45
  swing1Motion=Free
  swing2LimitDegrees=45
  swing2Motion=Free
  twistLimitDegrees=45
  twistMotion=Free
 }
 {
  bone1Name=head
  bone2Name=neck_02
  swing1LimitDegrees=45
  swing1Motion=Free
  swing2LimitDegrees=45
  swing2Motion=Free
  twistLimitDegrees=45
  twistMotion=Free
 }
 {
  bone1Name=dyn_cape_c_01
  bone2Name=dyn_cape_r_01
  swing1LimitDegrees=45
  swing1Motion=Free
  swing2LimitDegrees=45
  swing2Motion=Free
  twistLimitDegrees=45
  twistMotion=Free
 }
 {
  bone1Name=dyn_cape_c_01
  bone2Name=dyn_cape_l_01
  swing1LimitDegrees=45
  swing1Motion=Free
  swing2LimitDegrees=45
  swing2Motion=Free
  twistLimitDegrees=45
  twistMotion=Free
 }
]
