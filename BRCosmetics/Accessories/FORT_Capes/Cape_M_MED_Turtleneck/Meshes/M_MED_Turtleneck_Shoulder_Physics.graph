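# Physics asset for the shoulder piece of the M_MED_Turtleneck cape cosmetic, covering the character's right shoulder.
# Capsules on the upper spine, right clavicle and upper arm form a body chain joined by angularly free constraints.
# Six dynamic-bone chains (a to f) branch from the upper-arm twist bone, each a tiny master capsule followed by two boxes.
# Tight swing and even tighter twist limits within each chain restrict how far the segments bend.
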PhysicsAsset /BRCosmetics/Accessories/FORT_Capes/Cape_M_MED_Turtleneck/Meshes/M_MED_Turtleneck_Shoulder_Physics
bodies {
 clavicle_r [{shapeType=Capsule, length=12.8431, radius=6.3669, center=<-12.5433 0.1632 0>, rotation=<rot 90 0 0>}]
 dyn_shoulder_a_01 [{shapeType=Box, extentX=5.9501, extentY=9.039, extentZ=3.9143, center=<3.0274 0 0>}]
 dyn_shoulder_a_02 [{shapeType=Box, extentX=5.9501, extentY=9.039, extentZ=3.9143, center=<3.0274 0 0>}]
 dyn_shoulder_a_MSTR [{shapeType=Capsule, length=0.505, radius=0.505}]
 dyn_shoulder_b_01 [{shapeType=Box, extentX=5.9501, extentY=6.3706, extentZ=3.2695, center=<3.0941 0 0.0001>, rotation=<rot 0 0 15.0611>}]
 dyn_shoulder_b_02 [{shapeType=Box, extentX=5.9501, extentY=6.3706, extentZ=3.2695, center=<3.0941 0 0.0001>, rotation=<rot 0 0 15.0611>}]
 dyn_shoulder_b_MSTR [{shapeType=Capsule, length=0.505, radius=0.505}]
 dyn_shoulder_c_01 [{shapeType=Box, extentX=5.9501, extentY=13.7299, extentZ=4.9266, center=<3.0311 0 0>}]
 dyn_shoulder_c_02 [{shapeType=Box, extentX=5.9501, extentY=13.7299, extentZ=4.9266, center=<3.0311 0 0>}]
 dyn_shoulder_c_MSTR [{shapeType=Capsule, length=0.505, radius=0.505}]
 dyn_shoulder_d_01 [{shapeType=Box, extentX=5.9501, extentY=9.6067, extentZ=6.5229, center=<3.0464 0 -0.0001>}]
 dyn_shoulder_d_02 [{shapeType=Box, extentX=5.9501, extentY=9.6067, extentZ=6.5229, center=<3.0464 0 -0.0001>}]
 dyn_shoulder_d_MSTR [{shapeType=Capsule, length=0.505, radius=0.505}]
 dyn_shoulder_e_01 [{shapeType=Box, extentX=5.9501, extentY=6.3706, extentZ=6.1949, center=<2.8815 0 0>}]
 dyn_shoulder_e_02 [{shapeType=Box, extentX=5.9501, extentY=6.3706, extentZ=6.1949, center=<2.8815 0 0>}]
 dyn_shoulder_e_MSTR [{shapeType=Capsule, length=0.505, radius=0.505}]
 dyn_shoulder_f_01 [{shapeType=Box, extentX=5.9501, extentY=6.3706, extentZ=5.2218, center=<3.1341 0 0>}]
 dyn_shoulder_f_02 [{shapeType=Box, extentX=5.9501, extentY=6.3706, extentZ=5.2218, center=<3.1341 0 0>}]
 dyn_shoulder_f_MSTR [{shapeType=Capsule, length=0.505, radius=0.505}]
 spine_03 [{shapeType=Capsule, length=12.6492, radius=8.911}]
 spine_04 [{shapeType=Capsule, length=12.6492, radius=8.911}]
 spine_05 [{shapeType=Capsule, length=12.6492, radius=8.911}]
 upperarm_r [{shapeType=Capsule, length=15.0795, radius=5.0926, center=<-12.81 0.0001 0>, rotation=<rot 90 0 0>}]
 upperarm_twist_01_r [{shapeType=Capsule, length=24.6017, radius=6.4071, center=<-12.8125 0 0>, rotation=<rot 90 0 0>}]
}
constraints [
 {
  bone1Name=spine_04
  bone2Name=spine_03
  swing1LimitDegrees=45
  swing1Motion=Free
  swing2LimitDegrees=45
  swing2Motion=Free
  twistLimitDegrees=45
  twistMotion=Free
 }
 {
  bone1Name=spine_05
  bone2Name=spine_04
  swing1LimitDegrees=45
  swing1Motion=Free
  swing2LimitDegrees=45
  swing2Motion=Free
  twistLimitDegrees=45
  twistMotion=Free
 }
 {
  bone1Name=clavicle_r
  bone2Name=spine_05
  swing1LimitDegrees=45
  swing1Motion=Free
  swing2LimitDegrees=45
  swing2Motion=Free
  twistLimitDegrees=45
  twistMotion=Free
 }
 {
  bone1Name=upperarm_r
  bone2Name=clavicle_r
  swing1LimitDegrees=45
  swing1Motion=Free
  swing2LimitDegrees=45
  swing2Motion=Free
  twistLimitDegrees=45
  twistMotion=Free
 }
 {
  bone1Name=upperarm_twist_01_r
  bone2Name=upperarm_r
  swing1LimitDegrees=45
  swing1Motion=Free
  swing2LimitDegrees=45
  swing2Motion=Free
  twistLimitDegrees=45
  twistMotion=Free
 }
 {
  bone1Name=dyn_shoulder_a_MSTR
  bone2Name=upperarm_twist_01_r
  swing1LimitDegrees=45
  swing1Motion=Free
  swing2LimitDegrees=45
  swing2Motion=Free
  twistLimitDegrees=45
  twistMotion=Free
 }
 {
  bone1Name=dyn_shoulder_b_MSTR
  bone2Name=upperarm_twist_01_r
  swing1LimitDegrees=45
  swing1Motion=Free
  swing2LimitDegrees=45
  swing2Motion=Free
  twistLimitDegrees=45
  twistMotion=Free
 }
 {
  bone1Name=dyn_shoulder_c_MSTR
  bone2Name=upperarm_twist_01_r
  swing1LimitDegrees=45
  swing1Motion=Free
  swing2LimitDegrees=45
  swing2Motion=Free
  twistLimitDegrees=45
  twistMotion=Free
 }
 {
  bone1Name=dyn_shoulder_d_MSTR
  bone2Name=upperarm_twist_01_r
  swing1LimitDegrees=45
  swing1Motion=Free
  swing2LimitDegrees=45
  swing2Motion=Free
  twistLimitDegrees=45
  twistMotion=Free
 }
 {
  bone1Name=dyn_shoulder_e_MSTR
  bone2Name=upperarm_twist_01_r
  swing1LimitDegrees=45
  swing1Motion=Free
  swing2LimitDegrees=45
  swing2Motion=Free
  twistLimitDegrees=45
  twistMotion=Free
 }
 {
  bone1Name=dyn_shoulder_f_MSTR
  bone2Name=upperarm_twist_01_r
  swing1LimitDegrees=45
  swing1Motion=Free
  swing2LimitDegrees=45
  swing2Motion=Free
  twistLimitDegrees=45
  twistMotion=Free
 }
 {bone1Name=dyn_shoulder_a_02, bone2Name=dyn_shoulder_a_01, swing1LimitDegrees=20, swing2LimitDegrees=20, twistLimitDegrees=5}
 {bone1Name=dyn_shoulder_b_02, bone2Name=dyn_shoulder_b_01, swing1LimitDegrees=20, swing2LimitDegrees=20, twistLimitDegrees=5}
 {bone1Name=dyn_shoulder_c_01, bone2Name=dyn_shoulder_c_MSTR, swing1LimitDegrees=20, swing2LimitDegrees=20, twistLimitDegrees=5}
 {bone1Name=dyn_shoulder_d_02, bone2Name=dyn_shoulder_d_01, swing1LimitDegrees=20, swing2LimitDegrees=20, twistLimitDegrees=5}
 {bone1Name=dyn_shoulder_e_02, bone2Name=dyn_shoulder_e_01, swing1LimitDegrees=20, swing2LimitDegrees=20, twistLimitDegrees=5}
 {bone1Name=dyn_shoulder_d_01, bone2Name=dyn_shoulder_d_MSTR, swing1LimitDegrees=20, swing2LimitDegrees=20, twistLimitDegrees=5}
 {bone1Name=dyn_shoulder_a_01, bone2Name=dyn_shoulder_a_MSTR, swing1LimitDegrees=20, swing2LimitDegrees=20, twistLimitDegrees=5}
 {bone1Name=dyn_shoulder_b_01, bone2Name=dyn_shoulder_b_MSTR, swing1LimitDegrees=20, swing2LimitDegrees=20, twistLimitDegrees=5}
 {bone1Name=dyn_shoulder_c_02, bone2Name=dyn_shoulder_c_01, swing1LimitDegrees=20, swing2LimitDegrees=20, twistLimitDegrees=5}
 {bone1Name=dyn_shoulder_e_01, bone2Name=dyn_shoulder_e_MSTR, swing1LimitDegrees=20, swing2LimitDegrees=20, twistLimitDegrees=5}
 {bone1Name=dyn_shoulder_f_01, bone2Name=dyn_shoulder_f_MSTR, swing1LimitDegrees=20, swing2LimitDegrees=20, twistLimitDegrees=5}
 {bone1Name=dyn_shoulder_f_02, bone2Name=dyn_shoulder_f_01, swing1LimitDegrees=20, swing2LimitDegrees=20, twistLimitDegrees=5}
]
