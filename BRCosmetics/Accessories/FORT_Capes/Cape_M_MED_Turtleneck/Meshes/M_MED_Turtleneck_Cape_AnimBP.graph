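# Animation Blueprint for the Turtleneck cape cosmetic, adding secondary motion to cape and shoulder bones over the copied parent pose.
# RigidBody chains with separate physics setups blend by front-end, skydiving, crouch, and backbling-or-melee states.
# Bone constraints then tie shoulder bones to the right arm and cape columns together, with additive clearance offsets.
# The event graph supplies gravity override, smoothed sim-space rotation, slope data, state flags and dynamics resets.
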AnimBlueprint /BRCosmetics/Accessories/FORT_Capes/Cape_M_MED_Turtleneck/Meshes/M_MED_Turtleneck_Cape_AnimBP
target_skeleton=&/M_MED_Turtleneck_Cape_Skeleton
graphs [
 AnimGraph {nodes [
   n1 = "Constraint - dyn_shoulder_f_MSTR (P)"(ComponentPose=Local To Component(LocalPose=Use cached pose 'rigidCache'()), Alpha=1, ConstraintWeights_0=1) {
    alphaBoolBlend {blendOption=Linear}
    boneToModify {boneName=dyn_shoulder_f_MSTR}
    constraintSetup [{targetBone {boneName=upperarm_twist_01_r}}]
    constraintWeights [1]
   }
   n2 = "Constraint - dyn_shoulder_e_MSTR (P)"(ComponentPose=@n1, Alpha=0.5, ConstraintWeights_0=1) {
    alphaBoolBlend {blendOption=Linear}
    boneToModify {boneName=dyn_shoulder_e_MSTR}
    constraintSetup [{targetBone {boneName=upperarm_twist_01_r}}]
    constraintWeights [1]
   }
   n3 = "Constraint - dyn_shoulder_d_MSTR (P)"(ComponentPose=@n2, Alpha=0.5, ConstraintWeights_0=1) {
    alphaBoolBlend {blendOption=Linear}
    boneToModify {boneName=dyn_shoulder_d_MSTR}
    constraintSetup [{targetBone {boneName=upperarm_twist_01_r}}]
    constraintWeights [1]
   }
   n4 = "Constraint - dyn_shoulder_c_MSTR (P)"(ComponentPose=@n3, Alpha=0.5, ConstraintWeights_0=1) {
    alphaBoolBlend {blendOption=Linear}
    boneToModify {boneName=dyn_shoulder_c_MSTR}
    constraintSetup [{targetBone {boneName=upperarm_twist_01_r}}]
    constraintWeights [1]
   }
   n5 = "Constraint - dyn_shoulder_b_MSTR (P)"(ComponentPose=@n4, Alpha=0.4, ConstraintWeights_0=1) {
    alphaBoolBlend {blendOption=Linear}
    boneToModify {boneName=dyn_shoulder_b_MSTR}
    constraintSetup [{targetBone {boneName=upperarm_twist_01_r}}]
    constraintWeights [1]
   }
   n6 = "Constraint - dyn_shoulder_a_MSTR (P)"(ComponentPose=@n5, Alpha=0.1, ConstraintWeights_0=1) {
    alphaBoolBlend {blendOption=Linear}
    boneToModify {boneName=dyn_shoulder_a_MSTR}
    constraintSetup [{targetBone {boneName=upperarm_twist_01_r}}]
    constraintWeights [1]
   }
   n7 = "Transform (Modify) Bone - Bone: dyn_root_simspace"(ComponentPose=@n6, Alpha=1, Translation="0,0,0", Rotation=Get sim_space_rotation() -> (sim_space_rotation:Rotator), Scale="1,1,1") {
    rotationMode=BMM_Replace
    rotationSpace=BCS_WorldSpace
    translationSpace=BCS_BoneSpace
    alphaBoolBlend {blendOption=Linear}
    alphaScaleBiasClamp {bInterpResult=true, interpSpeedDecreasing=2, interpSpeedIncreasing=2}
    boneToModify {boneName=dyn_root_simspace}
   }
   n8 = "Constraint - dyn_cape_l_01 (R)"(ComponentPose=@n7, Alpha=1, ConstraintWeights_0=1) {
    alphaBoolBlend {blendOption=Linear}
    boneToModify {boneName=dyn_cape_l_01}
    constraintSetup [{transformType=Rotation, targetBone {boneName=dyn_cape_c_01}}]
    constraintWeights [1]
   }
   n9 = "Constraint - dyn_cape_l_02 (R)"(ComponentPose=@n8, Alpha=0.95, ConstraintWeights_0=1) {
    alphaBoolBlend {blendOption=Linear}
    boneToModify {boneName=dyn_cape_l_02}
    constraintSetup [{transformType=Rotation, targetBone {boneName=dyn_cape_c_02}}]
    constraintWeights [1]
   }
   n10 = "Constraint - dyn_cape_l_03 (R)"(ComponentPose=@n9, Alpha=0.95, ConstraintWeights_0=1) {
    alphaBoolBlend {blendOption=Linear}
    boneToModify {boneName=dyn_cape_l_03}
    constraintSetup [{transformType=Rotation, targetBone {boneName=dyn_cape_c_03}}]
    constraintWeights [1]
   }
   n11 = "Constraint - dyn_cape_l_04 (R)"(ComponentPose=@n10, Alpha=0.95, ConstraintWeights_0=1) {
    alphaBoolBlend {blendOption=Linear}
    boneToModify {boneName=dyn_cape_l_04}
    constraintSetup [{transformType=Rotation, targetBone {boneName=dyn_cape_c_04}}]
    constraintWeights [1]
   }
   n12 = "Constraint - dyn_cape_l_05 (R)"(ComponentPose=@n11, Alpha=0.95, ConstraintWeights_0=1) {
    alphaBoolBlend {blendOption=Linear}
    boneToModify {boneName=dyn_cape_l_05}
    constraintSetup [{transformType=Rotation, targetBone {boneName=dyn_cape_c_05}}]
    constraintWeights [1]
   }
   n13 = "Constraint - dyn_cape_l_06 (R)"(ComponentPose=@n12, Alpha=0.95, ConstraintWeights_0=1) {
    alphaBoolBlend {blendOption=Linear}
    boneToModify {boneName=dyn_cape_l_06}
    constraintSetup [{transformType=Rotation, targetBone {boneName=dyn_cape_c_06}}]
    constraintWeights [1]
   }
   n14 = "Constraint - dyn_cape_c_01 (RR)"!(ComponentPose=@n13, Alpha=1, ConstraintWeights_0=0.5, ConstraintWeights_1:float=0.5) -> (Pose) {
    alphaBoolBlend {blendOption=Linear}
    boneToModify {boneName=dyn_cape_c_01}
    constraintSetup [{transformType=Rotation, targetBone {boneName=dyn_cape_r_01}}, {transformType=Rotation, targetBone {boneName=dyn_cape_l_01}}]
    constraintWeights [1,1]
   }
   n15 = "Constraint - dyn_cape_c_02 (RR)"!(ComponentPose=@n14, Alpha=1, ConstraintWeights_0=0.5, ConstraintWeights_1:float=0.5) -> (Pose) {
    alphaBoolBlend {blendOption=Linear}
    boneToModify {boneName=dyn_cape_c_02}
    constraintSetup [{transformType=Rotation, targetBone {boneName=dyn_cape_r_02}}, {transformType=Rotation, targetBone {boneName=dyn_cape_l_02}}]
    constraintWeights [1,1]
   }
   n16 = "Constraint - dyn_cape_c_03 (RR)"!(ComponentPose=@n15, Alpha=1, ConstraintWeights_0=0.5, ConstraintWeights_1:float=0.5) -> (Pose) {
    alphaBoolBlend {blendOption=Linear}
    boneToModify {boneName=dyn_cape_c_03}
    constraintSetup [{transformType=Rotation, targetBone {boneName=dyn_cape_r_03}}, {transformType=Rotation, targetBone {boneName=dyn_cape_l_03}}]
    constraintWeights [1,1]
   }
   n17 = "Constraint - dyn_cape_c_04 (RR)"!(ComponentPose=@n16, Alpha=1, ConstraintWeights_0=0.5, ConstraintWeights_1:float=0.5) -> (Pose) {
    alphaBoolBlend {blendOption=Linear}
    boneToModify {boneName=dyn_cape_c_04}
    constraintSetup [{transformType=Rotation, targetBone {boneName=dyn_cape_r_04}}, {transformType=Rotation, targetBone {boneName=dyn_cape_l_04}}]
    constraintWeights [1,1]
   }
   n18 = "Constraint - dyn_cape_c_05 (RR)"!(ComponentPose=@n17, Alpha=1, ConstraintWeights_0=0.5, ConstraintWeights_1:float=0.5) -> (Pose) {
    alphaBoolBlend {blendOption=Linear}
    boneToModify {boneName=dyn_cape_c_05}
    constraintSetup [{transformType=Rotation, targetBone {boneName=dyn_cape_r_05}}, {transformType=Rotation, targetBone {boneName=dyn_cape_l_05}}]
    constraintWeights [1,1]
   }
   n19 = "Constraint - dyn_cape_c_06 (RR)"!(ComponentPose=@n18, Alpha=1, ConstraintWeights_0=0.5, ConstraintWeights_1:float=0.5) -> (Pose) {
    alphaBoolBlend {blendOption=Linear}
    boneToModify {boneName=dyn_cape_c_06}
    constraintSetup [{transformType=Rotation, targetBone {boneName=dyn_cape_r_06}}, {transformType=Rotation, targetBone {boneName=dyn_cape_l_06}}]
    constraintWeights [1,1]
   }
   n20 = "Constraint - dyn_cape_r_01 (R)"(ComponentPose=@n19, Alpha=1, ConstraintWeights_0=1) {
    alphaBoolBlend {blendOption=Linear}
    boneToModify {boneName=dyn_cape_r_01}
    constraintSetup [{transformType=Rotation, targetBone {boneName=dyn_cape_c_01}}]
    constraintWeights [1]
   }
   n21 = "Constraint - dyn_cape_r_02 (R)"(ComponentPose=@n20, Alpha=0.8, ConstraintWeights_0=1) {
    alphaBoolBlend {blendOption=Linear}
    boneToModify {boneName=dyn_cape_r_02}
    constraintSetup [{transformType=Rotation, targetBone {boneName=dyn_cape_c_02}}]
    constraintWeights [1]
   }
   n22 = "Constraint - dyn_cape_r_03 (R)"(ComponentPose=@n21, Alpha=0.8, ConstraintWeights_0=1) {
    alphaBoolBlend {blendOption=Linear}
    boneToModify {boneName=dyn_cape_r_03}
    constraintSetup [{transformType=Rotation, targetBone {boneName=dyn_cape_c_03}}]
    constraintWeights [1]
   }
   n23 = "Constraint - dyn_cape_r_04 (R)"(ComponentPose=@n22, Alpha=0.8, ConstraintWeights_0=1) {
    alphaBoolBlend {blendOption=Linear}
    boneToModify {boneName=dyn_cape_r_04}
    constraintSetup [{transformType=Rotation, targetBone {boneName=dyn_cape_c_04}}]
    constraintWeights [1]
   }
   n24 = "Constraint - dyn_cape_r_05 (R)"(ComponentPose=@n23, Alpha=0.8, ConstraintWeights_0=1) {
    alphaBoolBlend {blendOption=Linear}
    boneToModify {boneName=dyn_cape_r_05}
    constraintSetup [{transformType=Rotation, targetBone {boneName=dyn_cape_c_05}}]
    constraintWeights [1]
   }
   n25 = "Constraint - dyn_cape_r_06 (R)"(ComponentPose=@n24, Alpha=1, ConstraintWeights_0=1) {
    alphaBoolBlend {blendOption=Linear}
    boneToModify {boneName=dyn_cape_r_06}
    constraintSetup [{transformType=Rotation, targetBone {boneName=dyn_cape_c_06}}]
    constraintWeights [1]
   }
   n26 = "Transform (Modify) Bone - Bone: dyn_shoulder_d_MSTR"(ComponentPose=@n25, Alpha=1, Translation="0,0,1", Rotation="0,0,0", Scale="1,1,1") {
    rotationMode=BMM_Additive
    rotationSpace=BCS_BoneSpace
    scaleMode=BMM_Additive
    scaleSpace=BCS_BoneSpace
    translationMode=BMM_Additive
    translationSpace=BCS_BoneSpace
    alphaBoolBlend {blendOption=Linear}
    boneToModify {boneName=dyn_shoulder_d_MSTR}
    translation {z=1}
   }
   n27 = "Transform (Modify) Bone - Bone: dyn_shoulder_e_MSTR"(ComponentPose=@n26, Alpha=1, Translation="0,0,2", Rotation="0,0,0", Scale="1,1,1") {
    rotationMode=BMM_Additive
    rotationSpace=BCS_BoneSpace
    scaleMode=BMM_Additive
    scaleSpace=BCS_BoneSpace
    translationMode=BMM_Additive
    translationSpace=BCS_BoneSpace
    alphaBoolBlend {blendOption=Linear}
    boneToModify {boneName=dyn_shoulder_e_MSTR}
    translation {z=2}
   }
   n28 = "Transform (Modify) Bone - Bone: dyn_shoulder_f_MSTR"(ComponentPose=@n27, Alpha=1, Translation="0,0,1", Rotation="0,0,0", Scale="1,1,1") {
    rotationMode=BMM_Additive
    rotationSpace=BCS_BoneSpace
    scaleMode=BMM_Additive
    scaleSpace=BCS_BoneSpace
    translationMode=BMM_Additive
    translationSpace=BCS_BoneSpace
    alphaBoolBlend {blendOption=Linear}
    boneToModify {boneName=dyn_shoulder_f_MSTR}
    translation {z=1}
   }
   n29 = "Transform (Modify) Bone - Bone: dyn_cape01_l_MSTR"(ComponentPose=@n28, Alpha=1, Translation="0,0,1", Rotation="0,0,0", Scale="1,1,1") {
    rotationMode=BMM_Additive
    rotationSpace=BCS_BoneSpace
    scaleMode=BMM_Additive
    scaleSpace=BCS_BoneSpace
    translationMode=BMM_Additive
    translationSpace=BCS_BoneSpace
    alphaBoolBlend {blendOption=Linear}
    boneToModify {boneName=dyn_cape01_l_MSTR}
    translation {z=1}
   }
   n30 = "Transform (Modify) Bone - Bone: dyn_cape_c_MSTR"(ComponentPose=@n29, Alpha=1, Translation="0,0,1", Rotation="0,0,0", Scale="1,1,1") {
    rotationMode=BMM_Additive
    rotationSpace=BCS_BoneSpace
    scaleMode=BMM_Additive
    scaleSpace=BCS_BoneSpace
    translationMode=BMM_Additive
    translationSpace=BCS_BoneSpace
    alphaBoolBlend {blendOption=Linear}
    boneToModify {boneName=dyn_cape_c_MSTR}
    translation {z=1}
   }
   n31 = "Transform (Modify) Bone - Bone: dyn_cape_r_MSTR"(ComponentPose=@n30, Alpha=1, Translation="0,0,1", Rotation="0,0,0", Scale="1,1,1") {
    rotationMode=BMM_Additive
    rotationSpace=BCS_BoneSpace
    scaleMode=BMM_Additive
    scaleSpace=BCS_BoneSpace
    translationMode=BMM_Additive
    translationSpace=BCS_BoneSpace
    alphaBoolBlend {blendOption=Linear}
    boneToModify {boneName=dyn_cape_r_MSTR}
    translation {z=1}
   }
   n32 = "To Float (Boolean)"(InBool=Get isWearingBackbling())
   n33 = "Transform (Modify) Bone - Bone: dyn_cape_l_02"(ComponentPose=@n31, Alpha=@n32, Translation="0,0,0", Rotation="-5,0,0", Scale="1,1,1") {
    rotationMode=BMM_Additive
    rotationSpace=BCS_BoneSpace
    scaleMode=BMM_Additive
    scaleSpace=BCS_BoneSpace
    translationMode=BMM_Additive
    translationSpace=BCS_BoneSpace
    alphaBoolBlend {blendOption=Linear}
    boneToModify {boneName=dyn_cape_l_02}
    rotation {pitch=-5}
   }
   n34 = "Transform (Modify) Bone - Bone: dyn_cape_c_02"(ComponentPose=@n33, Alpha=@n32, Translation="0,0,0", Rotation="-5,0,0", Scale="1,1,1") {
    rotationMode=BMM_Additive
    rotationSpace=BCS_BoneSpace
    scaleMode=BMM_Additive
    scaleSpace=BCS_BoneSpace
    translationMode=BMM_Additive
    translationSpace=BCS_BoneSpace
    alphaBoolBlend {blendOption=Linear}
    boneToModify {boneName=dyn_cape_c_02}
    rotation {pitch=-5}
   }
   n35 = "Transform (Modify) Bone - Bone: dyn_cape_r_02"(ComponentPose=@n34, Alpha=@n32, Translation="0,0,0", Rotation="-5,0,0", Scale="1,1,1") {
    rotationMode=BMM_Additive
    rotationSpace=BCS_BoneSpace
    scaleMode=BMM_Additive
    scaleSpace=BCS_BoneSpace
    translationMode=BMM_Additive
    translationSpace=BCS_BoneSpace
    alphaBoolBlend {blendOption=Linear}
    boneToModify {boneName=dyn_cape_r_02}
    rotation {pitch=-5}
   }
   Output Pose(Result=Component To Local(ComponentPose=@n35)) {layerGroup=None}
   n36 = "Constraint - dyn_shoulder_a_MSTR (PP)"!(
    ComponentPose=Local To Component(LocalPose=Copy Pose From Mesh() {bCopyCurves=true, bUseAttachedParent=true})
    Alpha=1
    ConstraintWeights_0=0.8
    ConstraintWeights_1:float=0.2
   ) -> (Pose) {
    alphaBoolBlend {blendOption=Linear}
    boneToModify {boneName=dyn_shoulder_a_MSTR}
    constraintSetup [{targetBone {boneName=clavicle_r}}, {targetBone {boneName=upperarm_twist_01_r}}]
    constraintWeights [1,1]
   }
   n37 = "Constraint - dyn_shoulder_b_MSTR (PP)"!(ComponentPose=@n36, Alpha=1, ConstraintWeights_0=0.6, ConstraintWeights_1:float=0.4) -> (Pose) {
    alphaBoolBlend {blendOption=Linear}
    boneToModify {boneName=dyn_shoulder_b_MSTR}
    constraintSetup [{targetBone {boneName=clavicle_r}}, {targetBone {boneName=upperarm_twist_01_r}}]
    constraintWeights [1,1]
   }
   n38 = "Constraint - dyn_shoulder_b_MSTR (PP)"(ComponentPose=@n37, Alpha=1, ConstraintWeights_0=0.5, ConstraintWeights_1=0.5) {
    alphaBoolBlend {blendOption=Linear}
    boneToModify {boneName=dyn_shoulder_b_MSTR}
    constraintSetup [{targetBone {boneName=clavicle_r}}, {targetBone {boneName=upperarm_twist_01_r}}]
    constraintWeights [1,1]
   }
   n39 = "Transform (Modify) Bone - Bone: dyn_shoulder_b_MSTR"(ComponentPose=@n38, Alpha=1, Translation="0,0,10", Rotation="0,0,0", Scale="1,1,1") {
    rotationMode=BMM_Additive
    rotationSpace=BCS_BoneSpace
    scaleMode=BMM_Additive
    scaleSpace=BCS_BoneSpace
    translationMode=BMM_Additive
    translationSpace=BCS_BoneSpace
    alphaBoolBlend {blendOption=Linear}
    boneToModify {boneName=dyn_shoulder_b_MSTR}
    translation {z=10}
   }
   n40 = "Transform (Modify) Bone - Bone: dyn_root_slope"!(
    ComponentPose=@n39
    bAlphaBoolEnabled:bool=NOT Boolean(A=Get bCanPlayCustomAnimations() -> (bCanPlayCustomAnimations:bool))
    Translation=Get slope_root_trans() -> (slope_root_trans:Vector)
    Rotation=Get slope_foot_rot() -> (slope_foot_rot:Rotator)
    Scale="1,1,1"
   ) -> (Pose) {
    alphaInputType=Bool
    rotationMode=BMM_Replace
    rotationSpace=BCS_WorldSpace
    translationMode=BMM_Replace
    translationSpace=BCS_WorldSpace
    alphaBoolBlend {blendOption=Linear}
    boneToModify {boneName=dyn_root_slope}
   }
   n41 = "Transform (Modify) Bone - Bone: dyn_root_simspace"(ComponentPose=@n40, Alpha=1, Translation="0,0,0", Rotation=Get root_space_rotation() -> (root_space_rotation:Rotator), Scale="1,1,1") {
    rotationMode=BMM_Replace
    rotationSpace=BCS_WorldSpace
    translationSpace=BCS_BoneSpace
    alphaBoolBlend {blendOption=Linear}
    alphaScaleBiasClamp {bInterpResult=true, interpSpeedDecreasing=2, interpSpeedIncreasing=2}
    boneToModify {boneName=dyn_root_simspace}
   }
   n42 = "Transform (Modify) Bone - Bone: dyn_crouch_collision"(
    ComponentPose=@n41
    Alpha=Get crouch alpha inv() -> (crouch alpha inv:double)
    Translation="0.000,0.000,-49.249"
    Rotation="0,0,0"
    Scale="1,1,1"
   ) {
    translationMode=BMM_Replace
    translationSpace=BCS_BoneSpace
    alphaBoolBlend {blendOption=Linear}
    alphaScaleBiasClamp {bInterpResult=true}
    boneToModify {boneName=dyn_crouch_collision}
    translation {z=-49.249}
   }
   Save cached pose 'initCache'(Pose=Component To Local(ComponentPose=@n42)) {cacheName=initCache}
   n43 = Trail controller - Bone: dyn_cape_c_10!(
    ComponentPose=Local To Component(LocalPose=Use cached pose 'rigidCache'())
    Alpha=1
    RelaxationSpeedScale:float=1
    FakeVelocity:Vector="0,0,0"
   ) -> (Pose) {
    bActorSpaceFakeVel=true
    bShowTrailLocation=true
    bUsePlanarLimit=true
    chainLength=6
    debugLifeTime=0.1
    maxDeltaTime=0.0333
    relaxationSpeedScale=1.2
    alphaScaleBiasClamp {bInterpResult=true, interpSpeedDecreasing=25, interpSpeedIncreasing=25}
    baseJoint {boneName=backpack_root}
    rotationLimits [
     {limitMax=<180 180 180>}
     {limitMax=<180 180 180>}
     {limitMax=<180 180 180>}
     {limitMax=<180 180 180>}
     {limitMax=<180 180 180>}
     {limitMax=<180 180 180>}
    ]
    rotationOffsets [<0 0 0>, <0 0 0>, <0 0 0>, <0 0 0>, <0 0 0>, <0 0 0>]
    trailBone {boneName=dyn_cape_c_10}
    trailRelaxationSpeed {editorCurveData {keys [<at 0 15 RCIM_Linear>,<at 0.5289 13.1838 RCIM_Linear>,<at 1 6 RCIM_Linear>]}}
   }
   n44 = Trail controller - Bone: dyn_cape_r_08!(ComponentPose=@n43, Alpha=1, RelaxationSpeedScale:float=1, FakeVelocity:Vector="0,0,0") -> (Pose) {
    bActorSpaceFakeVel=true
    bShowTrailLocation=true
    bUsePlanarLimit=true
    chainLength=5
    debugLifeTime=0.1
    maxDeltaTime=0.0333
    relaxationSpeedScale=1.2
    alphaScaleBiasClamp {bInterpResult=true, interpSpeedDecreasing=25, interpSpeedIncreasing=25}
    baseJoint {boneName=backpack_root}
    rotationLimits [{limitMax=<180 180 180>}, {limitMax=<180 180 180>}, {limitMax=<180 180 180>}, {limitMax=<180 180 180>}, {limitMax=<180 180 180>}]
    rotationOffsets [<0 0 0>, <0 0 0>, <0 0 0>, <0 0 0>, <0 0 0>]
    trailBone {boneName=dyn_cape_r_08}
    trailRelaxationSpeed {editorCurveData {keys [<at 0 15 RCIM_Linear>,<at 0.5289 13.1838 RCIM_Linear>,<at 1 6 RCIM_Linear>]}}
   }
   n45 = Trail controller - Bone: dyn_cape_l_06!(ComponentPose=@n44, Alpha=1, RelaxationSpeedScale:float=1, FakeVelocity:Vector="0,0,0") -> (Pose) {
    bActorSpaceFakeVel=true
    bShowTrailLocation=true
    bUsePlanarLimit=true
    chainLength=5
    debugLifeTime=0.1
    maxDeltaTime=0.0333
    relaxationSpeedScale=1.2
    alphaScaleBiasClamp {bInterpResult=true, interpSpeedDecreasing=25, interpSpeedIncreasing=25}
    baseJoint {boneName=backpack_root}
    rotationLimits [{limitMax=<180 180 180>}, {limitMax=<180 180 180>}, {limitMax=<180 180 180>}, {limitMax=<180 180 180>}, {limitMax=<180 180 180>}]
    rotationOffsets [<0 0 0>, <0 0 0>, <0 0 0>, <0 0 0>, <0 0 0>]
    trailBone {boneName=dyn_cape_l_06}
    trailRelaxationSpeed {editorCurveData {keys [<at 0 15 RCIM_Linear>,<at 0.5289 13.1838 RCIM_Linear>,<at 1 6 RCIM_Linear>]}}
   }
   Save cached pose 'trailCache'(Pose=Component To Local(ComponentPose=@n45)) {cacheName=trailCache}
   n46 = "Transform (Modify) Bone - Bone: dyn_cape01_l_MSTR"(ComponentPose=Local To Component(LocalPose=Use cached pose 'initCache'()), Alpha=0.5, Translation="0,0,3", Rotation="-5,15,0", Scale="1,1,1") {
    rotationMode=BMM_Additive
    rotationSpace=BCS_BoneSpace
    scaleMode=BMM_Additive
    scaleSpace=BCS_BoneSpace
    translationMode=BMM_Additive
    translationSpace=BCS_BoneSpace
    alphaBoolBlend {blendOption=Linear}
    boneToModify {boneName=dyn_cape01_l_MSTR}
    rotation {pitch=-5, yaw=15}
    translation {z=3}
   }
   n47 = "Transform (Modify) Bone - Bone: dyn_cape_r_MSTR"(ComponentPose=@n46, Alpha=0.5, Translation="0,0,1", Rotation="-6,15,0", Scale="1,1,1") {
    rotationMode=BMM_Additive
    rotationSpace=BCS_BoneSpace
    scaleMode=BMM_Additive
    scaleSpace=BCS_BoneSpace
    translationMode=BMM_Additive
    translationSpace=BCS_BoneSpace
    alphaBoolBlend {blendOption=Linear}
    boneToModify {boneName=dyn_cape_r_MSTR}
    rotation {pitch=-6, yaw=15}
    translation {z=1}
   }
   n48 = "Transform (Modify) Bone - Bone: dyn_cape_c_MSTR"(ComponentPose=@n47, Alpha=0.5, Translation="0,0,2", Rotation="-3,15,0", Scale="1,1,1") {
    rotationMode=BMM_Additive
    rotationSpace=BCS_BoneSpace
    scaleMode=BMM_Additive
    scaleSpace=BCS_BoneSpace
    translationMode=BMM_Additive
    translationSpace=BCS_BoneSpace
    alphaBoolBlend {blendOption=Linear}
    boneToModify {boneName=dyn_cape_c_MSTR}
    rotation {pitch=-3, yaw=15}
    translation {z=2}
   }
   n49 = Get GravityOverride()
   n50 = RigidBody!(ComponentPose=@n48, Alpha, OverrideWorldGravity:Vector=@n49, ExternalForce) -> (Pose) {
    bClampLinearTranslationLimitToRefPose=true
    bOverrideWorldGravity=true
    overridePhysicsAsset=&/M_MED_Turtleneck_Cape_WithBackpack_Physics
    simulationSpace=BaseBoneSpace
    alphaBoolBlend {blendOption=Linear}
    baseBoneRef {boneName=pelvis}
    componentAppliedLinearAccClamp=<3000 3000 3000>
    componentLinearAccScale {z=0.3}
    componentLinearVelScale {z=0.3}
    overrideWorldGravity {z=-980}
    simSpaceSettings {
     maxAngularAcceleration=1000
     maxAngularVelocity=3
     maxLinearAcceleration=700
     maxLinearVelocity=500
     velocityScaleZ=0
     worldAlpha=1
     externalLinearDragV=<1 1>
    }
   }
   n51 = RigidBody(ComponentPose=@n50, OverrideWorldGravity=@n49) {
    bClampLinearTranslationLimitToRefPose=true
    bOverrideWorldGravity=true
    overridePhysicsAsset=&/M_MED_Turtleneck_Shoulder_Physics
    simulationSpace=BaseBoneSpace
    alphaBoolBlend {blendOption=Linear}
    baseBoneRef {boneName=dyn_root_simspace}
    componentAppliedLinearAccClamp=<3000 3000 3000>
    componentLinearAccScale {z=0.5}
    componentLinearVelScale {z=0.5}
    overrideWorldGravity {z=-980}
    simSpaceSettings {
     maxAngularAcceleration=1000
     maxAngularVelocity=3
     maxLinearAcceleration=700
     maxLinearVelocity=500
     velocityScaleZ=0
     worldAlpha=1
     externalLinearDragV=<1 1>
    }
   }
   n52 = RigidBody(ComponentPose=Local To Component(LocalPose=Use cached pose 'initCache'()), OverrideWorldGravity=@n49) {
    bClampLinearTranslationLimitToRefPose=true
    bOverrideWorldGravity=true
    overridePhysicsAsset=&/M_MED_Turtleneck_Cape_Physics
    simulationSpace=BaseBoneSpace
    alphaBoolBlend {blendOption=Linear}
    baseBoneRef {boneName=dyn_root_simspace}
    componentAppliedLinearAccClamp=<3000 3000 3000>
    componentLinearAccScale {z=0.45}
    componentLinearVelScale {z=0.45}
    overrideWorldGravity {z=-980}
    simSpaceSettings {
     maxAngularAcceleration=1000
     maxAngularVelocity=3
     maxLinearAcceleration=700
     maxLinearVelocity=500
     velocityScaleZ=0
     worldAlpha=1
     externalLinearDragV=<1 1>
    }
   }
   n53 = RigidBody(ComponentPose=@n52, OverrideWorldGravity=@n49) {
    bClampLinearTranslationLimitToRefPose=true
    bOverrideWorldGravity=true
    overridePhysicsAsset=&M_MED_Turtleneck_Shoulder_Physics
    simulationSpace=BaseBoneSpace
    alphaBoolBlend {blendOption=Linear}
    baseBoneRef {boneName=dyn_root_simspace}
    componentAppliedLinearAccClamp=<3000 3000 3000>
    componentLinearAccScale=<0.4 0.4 0.5>
    componentLinearVelScale=<0.4 0.4 0.5>
    overrideWorldGravity {z=-980}
    simSpaceSettings {
     maxAngularAcceleration=1000
     maxAngularVelocity=3
     maxLinearAcceleration=700
     maxLinearVelocity=500
     velocityScaleZ=0
     worldAlpha=1
     externalLinearDragV=<1 1>
    }
   }
   n54 = Get isWearingBackbling()
   n55 = Get isMelee() -> (isMelee:bool)
   n56 = RigidBody!(ComponentPose=Local To Component(LocalPose=Use cached pose 'initCache'()), Alpha, ExternalForce) -> (Pose) {
    bClampLinearTranslationLimitToRefPose=true
    bOverrideWorldGravity=true
    overridePhysicsAsset=&M_MED_Turtleneck_Cape_Physics
    simulationSpace=BaseBoneSpace
    alphaBoolBlend {blendOption=Linear}
    baseBoneRef {boneName=pelvis}
    componentAppliedLinearAccClamp=<0 0 0>
    overrideWorldGravity {z=-1200}
   }
   n57 = RigidBody!(ComponentPose=@n56, Alpha, OverrideWorldGravity:Vector="0,0,0", ExternalForce) -> (Pose) {
    bClampLinearTranslationLimitToRefPose=true
    bOverrideWorldGravity=true
    overridePhysicsAsset=&M_MED_Turtleneck_Shoulder_Physics
    simulationSpace=BaseBoneSpace
    alphaBoolBlend {blendOption=Linear}
    baseBoneRef {boneName=dyn_root_simspace}
    componentAppliedLinearAccClamp=<3000 3000 3000>
    componentLinearAccScale {z=0.5}
    componentLinearVelScale {z=0.5}
    overrideWorldGravity {z=-980}
    simSpaceSettings {
     maxAngularAcceleration=1000
     maxAngularVelocity=3
     maxLinearAcceleration=700
     maxLinearVelocity=500
     velocityScaleZ=0
     worldAlpha=1
     externalLinearDragV=<1 1>
    }
   }
   n58 = "Transform (Modify) Bone - Bone: dyn_cape01_l_MSTR"(ComponentPose=@n57, Alpha=1, Translation="0,0,2", Rotation="-1,0,0", Scale="1,1,1") {
    rotationMode=BMM_Additive
    rotationSpace=BCS_BoneSpace
    scaleMode=BMM_Additive
    translationMode=BMM_Additive
    translationSpace=BCS_BoneSpace
    alphaBoolBlend {blendOption=Linear}
    boneToModify {boneName=dyn_cape01_l_MSTR}
    rotation {pitch=-1}
    translation {z=2}
   }
   n59 = "Transform (Modify) Bone - Bone: dyn_cape_c_MSTR"(ComponentPose=@n58, Alpha=1, Translation="0,0,2", Rotation="0,0,0", Scale="1,1,1") {
    rotationMode=BMM_Additive
    rotationSpace=BCS_BoneSpace
    scaleMode=BMM_Additive
    translationMode=BMM_Additive
    translationSpace=BCS_BoneSpace
    alphaBoolBlend {blendOption=Linear}
    boneToModify {boneName=dyn_cape_c_MSTR}
    translation {z=2}
   }
   n60 = "Transform (Modify) Bone - Bone: dyn_cape_r_MSTR"(ComponentPose=@n59, Alpha=1, Translation="0,0,2", Rotation="-3,0,0", Scale="1,1,1") {
    rotationMode=BMM_Additive
    rotationSpace=BCS_BoneSpace
    scaleMode=BMM_Additive
    translationMode=BMM_Additive
    translationSpace=BCS_BoneSpace
    alphaBoolBlend {blendOption=Linear}
    boneToModify {boneName=dyn_cape_r_MSTR}
    rotation {pitch=-3}
    translation {z=2}
   }
   n61 = Get isWearingBackbling()
   n62 = Blend Poses by bool(
    BlendPose_0=Component To Local(ComponentPose=@n60)
    BlendPose_1=Blend Poses by bool(
     BlendPose_0=Component To Local(ComponentPose=@n51)
     BlendPose_1=Component To Local(ComponentPose=@n53)
     bActiveValue=OR Boolean(A=@n54, B=@n55)
    ) {blendType=Linear}
    BlendTime_0=1
    BlendTime_1=0.5
    bActiveValue=AND Boolean(A=Get bIsCrouching(), B=NOT Boolean(A=@n61))
   ) {blendType=Linear}
   n63 = Get GravityOverride()
   n64 = RigidBody(ComponentPose=Local To Component(LocalPose=Use cached pose 'initCache'()), OverrideWorldGravity=@n63) {
    bClampLinearTranslationLimitToRefPose=true
    bOverrideWorldGravity=true
    overridePhysicsAsset=&/M_MED_Turtleneck_Cape_Skydive_Physics
    simulationSpace=BaseBoneSpace
    alphaBoolBlend {blendOption=Linear}
    baseBoneRef {boneName=root}
    componentAppliedLinearAccClamp=<0 0 0>
    overrideWorldGravity {z=-50}
   }
   n65 = RigidBody(ComponentPose=@n64, OverrideWorldGravity=@n63) {
    bClampLinearTranslationLimitToRefPose=true
    bOverrideWorldGravity=true
    overridePhysicsAsset=&M_MED_Turtleneck_Shoulder_Physics
    simulationSpace=BaseBoneSpace
    alphaBoolBlend {blendOption=Linear}
    baseBoneRef {boneName=dyn_root_simspace}
    componentAppliedLinearAccClamp=<3000 3000 3000>
    componentLinearAccScale {z=0.5}
    componentLinearVelScale {z=0.5}
    overrideWorldGravity {z=-980}
    simSpaceSettings {
     maxAngularAcceleration=1000
     maxAngularVelocity=3
     maxLinearAcceleration=700
     maxLinearVelocity=500
     velocityScaleZ=0
     worldAlpha=1
     externalLinearDragV=<1 1>
    }
   }
   n66 = "Transform (Modify) Bone - Bone: dyn_cape01_l_MSTR"(ComponentPose=@n65, Alpha=1, Translation="0,0,0", Rotation="6,0,0", Scale="1,1,1") {
    rotationMode=BMM_Additive
    rotationSpace=BCS_BoneSpace
    scaleMode=BMM_Additive
    translationMode=BMM_Additive
    translationSpace=BCS_BoneSpace
    alphaBoolBlend {blendOption=Linear}
    boneToModify {boneName=dyn_cape01_l_MSTR}
    rotation {pitch=6}
   }
   n67 = "Transform (Modify) Bone - Bone: dyn_cape_c_MSTR"(ComponentPose=@n66, Alpha=1, Translation="0,0,0", Rotation="6,0,0", Scale="1,1,1") {
    rotationMode=BMM_Additive
    rotationSpace=BCS_BoneSpace
    scaleMode=BMM_Additive
    translationMode=BMM_Additive
    translationSpace=BCS_BoneSpace
    alphaBoolBlend {blendOption=Linear}
    boneToModify {boneName=dyn_cape_c_MSTR}
    rotation {pitch=6}
   }
   n68 = "Transform (Modify) Bone - Bone: dyn_cape_r_MSTR"(ComponentPose=@n67, Alpha=1, Translation="0,0,0", Rotation="6,0,0", Scale="1,1,1") {
    rotationMode=BMM_Additive
    rotationSpace=BCS_BoneSpace
    scaleMode=BMM_Additive
    translationMode=BMM_Additive
    translationSpace=BCS_BoneSpace
    alphaBoolBlend {blendOption=Linear}
    boneToModify {boneName=dyn_cape_r_MSTR}
    rotation {pitch=6}
   }
   n69 = Component To Local(ComponentPose=@n68)
   n70 = OR Boolean(A=Get bIsSkyDiving(), B=Get bIsSkydiveDiveMode())
   n71 = Get GravityOverride()
   n72 = RigidBody(ComponentPose=Local To Component(LocalPose=Use cached pose 'initCache'()), OverrideWorldGravity=@n71) {
    bClampLinearTranslationLimitToRefPose=true
    bOverrideWorldGravity=true
    overridePhysicsAsset=&M_MED_Turtleneck_Cape_Physics
    simulationSpace=BaseBoneSpace
    alphaBoolBlend {blendOption=Linear}
    baseBoneRef {boneName=root}
    componentAppliedLinearAccClamp=<3000 3000 3000>
    componentLinearAccScale {z=0.45}
    componentLinearVelScale {z=0.45}
    overrideWorldGravity {z=-980}
    simSpaceSettings {
     maxAngularAcceleration=1000
     maxAngularVelocity=3
     maxLinearAcceleration=700
     maxLinearVelocity=500
     velocityScaleZ=0
     worldAlpha=1
     externalLinearDragV=<1 1>
    }
   }
   n73 = RigidBody(ComponentPose=@n72, OverrideWorldGravity=@n71) {
    bClampLinearTranslationLimitToRefPose=true
    bOverrideWorldGravity=true
    overridePhysicsAsset=&M_MED_Turtleneck_Shoulder_Physics
    simulationSpace=BaseBoneSpace
    alphaBoolBlend {blendOption=Linear}
    baseBoneRef {boneName=root}
    componentAppliedLinearAccClamp=<3000 3000 3000>
    componentLinearAccScale=<0.4 0.4 0.5>
    componentLinearVelScale=<0.4 0.4 0.5>
    overrideWorldGravity {z=-980}
    simSpaceSettings {
     maxAngularAcceleration=1000
     maxAngularVelocity=3
     maxLinearAcceleration=700
     maxLinearVelocity=500
     velocityScaleZ=0
     worldAlpha=1
     externalLinearDragV=<1 1>
    }
   }
   Save cached pose 'rigidCache'(Pose=Blend Poses by bool(
     BlendPose_0=Component To Local(ComponentPose=@n73)
     BlendPose_1=Blend Poses by bool(BlendPose_0=@n69, BlendPose_1=@n62, BlendTime_0=1, BlendTime_1=0.2, bActiveValue=@n70) {blendType=Linear}
     BlendTime_0=1
     BlendTime_1=0.2
     bActiveValue=Get bIsInFrontEnd()
    ) {blendType=Linear}) {cacheName=rigidCache}
  ]}
 Complex Sim Space {nodes [
   n1 = vector + vector(
    A:Vector=GetSocketLocation!(self:Scene Component Object Reference=Get Skel_Mesh(), InSocketName:Name=root) -> (ReturnValue:Vector)
    B:Vector="0, 0,165"
   ) -> (ReturnValue:Vector) {is=title K2Node_CommutativeAssociativeBinaryOperator}
   n2 = Get target_sim_space() -> (target_sim_space:Rotator)
   n3 = Get root_space_rotation()
   n4 = Set root_space_rotation(
    execute=[@n12.then, @n13.then]
    root_space_rotation:Rotator=RInterpTo!(Current:Rotator=@n3, Target:Rotator=@n2, DeltaTime:float=0, InterpSpeed:float=Get InterpSpeed() -> (InterpSpeed:double)) -> (ReturnValue:Rotator)
   ) -> (Output_Get:Rotator)
   n5 = DrawDebugCoordinateSystem!(
    execute
    AxisLoc:Vector=@n1
    AxisRot:Rotator=@n4
    Scale:float=35
    Duration:float=0
    Thickness:float=3
    DepthPriority:EDrawDebugSceneDepthPriorityGroup Enum=World
   ) -> (then)
   Set inComponentSpaceLastFrame(execute=[@n5.then, @n4.then], inComponentSpaceLastFrame:bool=Get inComponentSpace() -> (inComponentSpace:bool)) -> (Output_Get:bool)
   > Branch(Condition="Equal (Rotator)"!(A:Rotator=@n3, B:Rotator=@n2, ErrorTolerance:float=0.0001) -> (ReturnValue:bool))
   > Set lockedOnSimSpace(lockedOnSimSpace:bool=true) -> (Output_Get:bool)
   > n6 = Set InterpSpeed(InterpSpeed:double=0) -> (Output_Get:double)
   Complex Sim Space() {is=title K2Node_FunctionEntry}
   > n7 = Sequence!(execute) -> (then_0, then_1, then_2:Exec, then_3:Exec)
   Reroute Node(InputPin=@n7.then_1)
   Reroute Node(InputPin=@n7.then_2)
   Set InterpSpeedMax(execute=@n7.then_3, InterpSpeedMax:double=0) -> (Output_Get:double)
   > n8 = Set inComponentSpace(inComponentSpace:bool="float > float"(A=Get PawnSpeed2D(), B=0) {is=title K2Node_CallFunction}) -> (Output_Get:bool)
   > n9 = Branch(Condition=@n8)
   n10 = GetOwningActor()
   > n11 = Set target_sim_space(target_sim_space:Rotator=Combine Rotators!(
     A:Rotator=Get Actor Rotation!(self:Actor Object Reference=@n10) -> (ReturnValue:Rotator) {is=title K2Node_CallFunction}
     B:Rotator="0,-90, 0"
    ) -> (ReturnValue:Rotator)) -> (Output_Get:Rotator)
   n12 = Branch(
    execute=[@n14.then, @n11.then]
    Condition="Equal (Boolean)"!(A:bool=Get inComponentSpaceLastFrame() -> (inComponentSpaceLastFrame:bool), B:bool=Get inComponentSpace()) -> (ReturnValue:bool)
   )
   Set lockedOnSimSpace(execute=@n12.else, lockedOnSimSpace=false)
   > n13 = Set InterpSpeed(InterpSpeed=Get InterpSpeedMax() -> (InterpSpeedMax:double))
   n14 = Set target_sim_space(execute=@n9.else, target_sim_space=Get root_rotation() -> (root_rotation:Rotator))
   Set root_rotation(
    execute=@n7.then_0
    root_rotation:Rotator=GetSocketRotation!(self:Scene Component Object Reference=Get Skel_Mesh(), InSocketName:Name=root) -> (ReturnValue:Rotator)
   ) -> (Output_Get:Rotator)
   DrawDebugCoordinateSystem(AxisLoc=@n1, AxisRot=Get root_rotation(), Scale=75, Duration=0, Thickness=1.5, DepthPriority=World)
   PrintString!(execute, InString="To String (Float)"(InDouble=@n6), bPrintToScreen, bPrintToLog, TextColor, Duration, Key) -> (then)
   n15 = GetOwningActor()
   n16 = Get Actor Location(self=@n15)
   n17 = vector * float(A=GetActorForwardVector!(self:Actor Object Reference=@n15) -> (ReturnValue:Vector), B=500)
   DrawDebugLine!(
    execute
    LineStart:Vector=vector + vector(A=@n16, B="0,0,87")
    LineEnd:Vector=vector + vector(A=vector + vector(A=@n16, B=@n17), B="0,0,87")
    LineColor:Linear Color Structure=(R=0.6348,G=0,B=0.6406,A=1)
    Duration:float=0
    Thickness:float=0.5
    DepthPriority:EDrawDebugSceneDepthPriorityGroup Enum=World
   ) -> (then)
  ]}
 EventGraph {nodes [
   n1 = Delay(Duration=20)
   ResetDynamics(execute=[@n1.then, @n3.then, @n2.then, @n11.then, @n12.then], InTeleportType=TeleportPhysics)
   > ResetDynamics(InTeleportType=ResetPhysics)
   n2 = Delay(Duration=20)
   Event OnParachuteOpened()
   > n3 = Delay(Duration=1.5)
   Event BlueprintUpdateAnimation()
   > n4 = Sequence!(execute) -> (then_0, then_1, then_2, then_3, then_4:Exec)
   n5 = ComplexSimSpaceFunction DEPRECATED!(
    execute=@n4.then_1
    CharacterPartAnimInstance:Custom Character Part Anim Instance Object Reference=Self-Reference()
    InterpSpeedMax:double=9
    PawnSpeedMax:double=20
    RInterpErrorTolerance:double=1
    DeltaTime:double=0
    DebugDraw:bool=false
    sim_space_rotation_in:Rotator=Get sim_space_rotation()
    inComponentSpaceIn:bool=Get InComponentSpace() -> (InComponentSpace:bool)
    inComponentSpaceLastFrameIn:bool=Get InComponentSpaceLastFrame() -> (InComponentSpaceLastFrame:bool)
    lockedOnSimSpaceIn:bool=Get LockedOnSimSpace() -> (LockedOnSimSpace:bool)
    InterpSpeedIn:double=Get InterpSpeed()
   ) -> (
    then
    sim_space_rotation:Rotator
    inComponentSpaceOut:bool
    inComponentSpaceLastFrameOut:bool
    lockedOnSimSpaceOut:bool
    InterpSpeedOut:double
   )
   > Set sim_space_rotation(sim_space_rotation:Rotator=@n5) -> (Output_Get:Rotator)
   > Set InComponentSpace(InComponentSpace:bool=@n5.inComponentSpaceOut) -> (Output_Get:bool)
   > Set InComponentSpaceLastFrame(InComponentSpaceLastFrame:bool=@n5.inComponentSpaceLastFrameOut) -> (Output_Get:bool)
   > Set LockedOnSimSpace(LockedOnSimSpace:bool=@n5.lockedOnSimSpaceOut) -> (Output_Get:bool)
   > Set InterpSpeed(InterpSpeed=@n5.InterpSpeedOut)
   n6 = Break Fort Anim Input Slope Info!(FortAnimInput_SlopeInfo:"Fort Anim Input Slope Info Structure (by ref)"=Get SlopeInfoInput() -> (SlopeInfoInput:Fort Anim Input Slope Info Structure)) -> (RootSlopeTranslation:Vector=~, LeftFootSlopeTranslation:Vector=~, RightFootSlopeTranslation:Vector=~, SlopeRotation:Rotator=~) {is=title K2Node_BreakStruct}
   Set slope_root_trans(execute=@n4.then_2, slope_root_trans:Vector=@n6) -> (Output_Get:Vector)
   > Set slope_foot_rot(slope_foot_rot:Rotator=@n6.SlopeRotation) -> (Output_Get:Rotator)
   n7 = GetOwnerSkeletalMeshComponent()
   n8 = BreakRotator(InRot=BreakTransform(InTransform=GetDeltaTransformFromRefPose(self=@n7, BoneName=upperarm_r, BaseName=clavicle_r)).Rotation)
   Set ArmLift_r(execute=@n4.then_3, ArmLift_r:double=MapRangeClamped(Value="Absolute (Float)"(A=@n8.Yaw), InRangeA=90, InRangeB=120)) -> (Output_Get:double)
   n9 = Gravity Override State Function DEPRECATED(
    execute=@n4.then_0
    CharacterPartAnimInstance=Self-Reference()
    GameTimeInSeconds=GetGameTimeInSeconds()
    GravityOverrideParams=Get Gravity Override Params()
   )
   > Set GravityOverride(GravityOverride=@n9)
   n10 = Switch on EFortWeaponCoreAnimation!(
    execute=@n4.then_4
    Selection:EFortWeaponCoreAnimation Enum=GetWeaponCoreAnim!(self:Fort Player Animation Interface Object Reference=GetOwnerAnimationInterface!(self:Custom Character Part Anim Instance Object Reference) -> (ReturnValue:Fort Player Animation Interface Object Reference)) -> (ReturnValue:EFortWeaponCoreAnimation Enum)
   ) -> (
    Melee:Exec
    Pistol:Exec
    Shotgun:Exec
    PaperBlueprint:Exec
    Rifle:Exec
    MeleeOneHand:Exec
    MachinePistol:Exec
    RocketLauncher:Exec
    GrenadeLauncher:Exec
    GoingCommando:Exec
    AssaultRifle:Exec
    TacticalShotgun:Exec
    SniperRifle:Exec
    TrapPlacement:Exec
    ShoulderLauncher:Exec
    AbilityDecoTool:Exec
    Crossbow:Exec
    C4:Exec
    RemoteControl:Exec
    DualWield:Exec
    AR_BullPup:Exec
    AR_ForwardGrip:Exec
    MedPackPaddles:Exec
    SMG_P90:Exec
    AR_DrumGun:Exec
    Consumable_Small:Exec
    Consumable_Large:Exec
    Balloon:Exec
    MountedTurret:Exec
    CreativeTool:Exec
    ExplosiveBow:Exec
    AshtonIndigo:Exec
    AshtonChicago:Exec
    MeleeDualWield:Exec
    Unarmed:Exec
   )
   Set isMelee(
    execute=[
     @n10.AR_BullPup
     @n10.AR_DrumGun
     @n10.AR_ForwardGrip
     @n10.AbilityDecoTool
     @n10.AshtonChicago
     @n10.AshtonIndigo
     @n10.AssaultRifle
     @n10.Balloon
     @n10.C4
     @n10.Consumable_Large
     @n10.Consumable_Small
     @n10.CreativeTool
     @n10.Crossbow
     @n10.DualWield
     @n10.ExplosiveBow
     @n10.GoingCommando
     @n10.GrenadeLauncher
     @n10.MachinePistol
     @n10.MedPackPaddles
     @n10.MountedTurret
     @n10.PaperBlueprint
     @n10.Pistol
     @n10.RemoteControl
     @n10.Rifle
     @n10.RocketLauncher
     @n10.SMG_P90
     @n10.Shotgun
     @n10.ShoulderLauncher
     @n10.SniperRifle
     @n10.TacticalShotgun
     @n10.TrapPlacement
     @n10.Unarmed
    ]
    isMelee:bool=false
   ) -> (Output_Get:bool)
   Set isMelee(execute=[@n10.Melee, @n10.MeleeDualWield, @n10.MeleeOneHand], isMelee=true)
   Event BlueprintInitializeAnimation()
   > Set Skel_Mesh(Skel_Mesh=GetOwnerSkeletalMeshComponent())
   > Set InterpSpeed(InterpSpeed=Get InterpSpeedMax())
   n11 = Event OnEndSkydiving()
   n12 = Event OnBeginSkydiving()
   Event OnCharacterCustomizationLoaded()
   > n13 = Cast To PlayerPawn_Athena(Object=TryGetPawnOwner())
   > Set Body_Mesh(Body_Mesh=GetSkeletalMeshForPartType(self=@n13))
   n14 = CheckForBackbling()
   > Set isWearingBackbling(isWearingBackbling=OR Boolean(A=@n14, B=@n14.bIsBackblingLong))
   Event OnBeginCrouching()
   > Set crouch alpha(crouch alpha:double=1) -> (Output_Get:double)
   > Set crouch alpha inv(crouch alpha inv:double=0) -> (Output_Get:double)
   Event OnEndCrouching()
   > Set crouch alpha(crouch alpha=0)
   > Set crouch alpha inv(crouch alpha inv=1)
   TryGetPawnOwner()
   Event OnEndPlayingEmote() {is=title K2Node_Event}
   Event OnBeginPlayingEmote() {is=title K2Node_Event}
  ]}
]
variables [
 ArmLift_r {type=number}
 Body_Mesh {type=SkeletalMeshComponent}
 crouch alpha {type=number}
 crouch alpha inv {type=number}
 Gravity Override Params {
  type=GravityOverrideParamsStruct
  default {
   skydive_DivingDown {gravityOverride=<-5000 0 0>, windAmplitude=<200 -500 400>, windOffset=<0 -600 400>}
   skydive_FloatingDown {gravityOverride=<0 0 600>}
   skydive_GoingUp {gravityOverride=<0 0 -100>, windOffset=<0 -600 400>}
   skydive_Parachuting {windOffset=<0 -600 400>}
   swimming {gravityOverride=<0 0 0>}
  }
 }
 GravityOverride {type=Vector}
 InComponentSpace {type=boolean}
 InComponentSpaceLastFrame {type=boolean}
 InterpSpeed {type=number}
 InterpSpeedMax {type=number}
 isCrouching {type=number}
 isMelee {type=boolean}
 isSkydiving {type=number}
 isSkydivingINV {type=number}
 isSprinting {type=number}
 isWearingBackbling {type=boolean}
 LockedOnSimSpace {type=boolean}
 root_rotation {type=Rotator}
 root_space_rotation {type=Rotator}
 sim_space_rotation {type=Rotator}
 Skel_Mesh {type=SkeletalMeshComponent}
 slope_foot_rot {type=Rotator}
 slope_root_trans {type=Vector}
 target_sim_space {type=Rotator}
]
